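# Sound cue for the DragonPop buying sound, routed to the Fort_UI sound class.
# A root Mixer plays three wave layers together, with the Buying_A layer boosted and the Buying_01 and Buying_03 layers attenuated.
# A concurrency override caps simultaneous instances, prevents new ones at the limit, and sets a short retrigger time.
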SoundCue /DragonPopGameplayContent/Assets/Audio/SC_DragonPop_Buying
root=Mixer_0
nodes [
 {id=Mixer_0, children [WavePlayer_5, WavePlayer_9, WavePlayer_11], properties {inputVolume [1.2,0.6,0.5]}}
 {id=WavePlayer_11, properties {soundWaveAssetPtr=&/DragonPop_Buying_03}}
 {id=WavePlayer_5, properties {soundWaveAssetPtr=&/DragonPop_Buying_A}}
 {id=WavePlayer_9, properties {soundWaveAssetPtr=&/DragonPop_Buying_01}}
]
properties {
 bOverrideConcurrency=true
 maxDistance=2097152
 soundClassObject=&Fort_UI
 volumeMultiplier=1
 concurrencyOverrides {maxCount=5, resolutionRule=PreventNew, retriggerTime=0.2}
}
