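# Sound cue for the DragonPop interact sound, playing the InitPurchase audio, likely heard when a purchase begins.
# A single wave player feeds the output with no randomization or mixing, slightly boosted in volume under the loot-collect sound class.
# It uses the shared mono item-interact attenuation and overrides concurrency to allow one instance, blocking new plays and rapid retriggers.
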SoundCue /DragonPopGameplayContent/Assets/Audio/SC_DragonPop_Interact
root=WavePlayer_0
nodes [{id=WavePlayer_0, properties {soundWaveAssetPtr=&/SC_DragonPop_InitPurchase}}]
properties {
 attenuationSettings=/Game/Sounds/Attenuation/Fort_Item_Interact_Mono_Att
 bOverrideConcurrency=true
 maxDistance=1500
 soundClassObject=&Fort_Loot_Collect
 volumeMultiplier=1.2
 concurrencyOverrides {maxCount=1, resolutionRule=PreventNew, retriggerTime=0.2}
}
