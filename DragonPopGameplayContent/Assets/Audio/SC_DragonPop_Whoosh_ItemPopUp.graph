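# Sound cue for the DragonPop item pop-up whoosh.
# A wave player for the whoosh sound feeds a modulator at the root.
# It plays at half volume on the critical in-game UI sound class, with Fortnite's item-interact mono attenuation.
# An overridden concurrency rule blocks new instances rather than cutting off playing ones.
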SoundCue /DragonPopGameplayContent/Assets/Audio/SC_DragonPop_Whoosh_ItemPopUp
root=Modulator_0
nodes [
 {id=Modulator_0, children [WavePlayer_1]}
 {id=WavePlayer_1, properties {soundWaveAssetPtr=&/DragonPop_Whoosh_ItemPopUp}}
]
properties {
 attenuationSettings=/Game/Sounds/Attenuation/Fort_Item_Interact_Mono_Att
 bOverrideConcurrency=true
 maxDistance=1500
 soundClassObject=&Fort_UI_InGame_Critical
 volumeMultiplier=0.5
 concurrencyOverrides {resolutionRule=PreventNew}
}
